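# Niagara skydiving contrail for the SteedShimmer cosmetic, trailing shimmering ribbons, bubbles, stars and pastel clouds from two skeletal mesh bones or sockets.
# MainTrail rebuilds four spiralling ribbons each update from partitioned particle indices and sine offsets, with quality-scaled point counts.
# Sprite emitters add bone-attached puffs, bubbles and timed large-cloud bursts, while a mesh emitter adds rotating colored streaks.
# ParaGlideLeanAlpha bends the motion, and time-of-day color tints every emitter.
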NiagaraSystem /BRCosmetics/Athena/Effects/Particles/Skydiving/NS_Contrail_SteedShimmer/NS_Contrail_SteedShimmer
dependencies {
 usedDataInterfaces [
  &NiagaraDataInterfaceArrayFloat
  &NiagaraDataInterfaceCurve
  &NiagaraDataInterfaceArrayColor
  &NiagaraDataInterfaceSpriteRendererInfo
  &NiagaraDataInterfaceArrayInt32
  &NiagaraDataInterfaceVectorCurve
  &NiagaraDataInterfaceArrayFloat3
  &NiagaraDataInterfaceColorCurve
  &NiagaraDataInterfaceSkeletalMesh
 ]
 usedDynamicInputs [
  &Boolean_OrOperation
  &SetBoolByFloatComparison
  &MakeFloatFromVector
  &ConvertVectorToPosition
  &FNTimeOfDayColor
 ]
 usedModules [
  &self:SystemSpawnScript.SystemScriptSource.SystemScriptGraph.NiagaraNodeAssignment_3.SetVariables_#1
  &self:SystemSpawnScript.SystemScriptSource.SystemScriptGraph.NiagaraNodeAssignment_4.SetVariables_#2
  &SystemState
  /Game/Effects/Niagara/Modules/Contrails/FN_Contrail_Setup
  &FN_Integrator
  &self:SystemSpawnScript.SystemScriptSource.SystemScriptGraph.NiagaraNodeAssignment_2.SetVariables_#3
  &self:SystemSpawnScript.SystemScriptSource.SystemScriptGraph.NiagaraNodeAssignment_0.SetVariables_#4
  &self:SystemSpawnScript.SystemScriptSource.SystemScriptGraph.NiagaraNodeAssignment_7.SetVariables_#5
 ]
}
emitters {
 Bubbles_Stars {
  emitterUpdateScript [
   &EmitterState(
    Delay First Loop Only=true
    Inactive Response=::NewEnumerator1
    Life Cycle Mode=::NewEnumerator1
    Loop Delay=0.3
    Loop Duration=1
    UseLoopDelay=true
   )
   &SpawnRate(SpawnGroup=2, SpawnRate=40, Use Spawn Probability=false) {module=SpawnRate002}
   &SpawnRate(SpawnGroup=1, SpawnRate=40, Use Spawn Probability=false) {module=SpawnRate001}
   &SpawnRate(SpawnRate=60, Use Spawn Probability=false)
  ]
  particleSpawnScript [
   /SetVariables(Particles.BubbleIndex:NiagaraInt32=&RandomRangeInt(Minimum=0, Maximum=1)) {isSetParameters=true}
   /SetVariables(Particles.BubblePos:NiagaraInt32=&RandomRangeInt(Minimum=0, Maximum=2)) {isSetParameters=true}
   &InitializeParticle(
    Color Mode=::NewEnumerator0
    Lifetime=&MaskFloat(
     Mask Source=1
     Mask Parameter=$Particles.BubblePos
     Mask=1
     Masked Float=&MaskFloat(Mask Source=0, Mask Parameter=?, Mask=2, Masked Float=0.5, Passthrough Float=1)
     Passthrough Float=1
    )
    Sprite Rotation Mode=::NewEnumerator2
    Sprite Size Mode=::NewEnumerator3
    Uniform Sprite Size=&MaskFloat(
     Mask Source=0
     Mask Parameter=?
     Mask=2
     Masked Float=&RandomRangeFloat(Minimum=25, Maximum=30)
     Passthrough Float=&RandomRangeFloat(Minimum=10, Maximum=15)
    )
    UsePositionOffset=false
    Write Material Random=true
   )
   /SetVariables(Particles.VisibilityTag=$Emitter.SpawnGroup) {isSetParameters=true}
   /SetVariables(Particles.SubUVIndex:NiagaraFloat=&MakeFloatFromInt(Int=&RandomRangeInt(Minimum=0, Maximum=3))) {isSetParameters=true}
  ]
  particleUpdateScript [
   &ParticleState()
   &SpriteRotationRate(Rotation Rate=90)
   &SkeletalMeshLocation(Bone / Socket Index=$Particles.BubbleIndex, Bone Sampling Mode=::NewEnumerator4, MaskSpawnGroup=false, Skeletal Mesh=$System.Skeletal Mesh)
   /SetVariables(Particles.SideMovement:Vector3f=&MakeVector(
     X=&Multiply_Float(
      A=&Sine(
       Normalized Angle=&Add_Float(
        A=&Multiply_Float(
         A=&Add_Float(A=0, B=$Particles.Age)
         B=&SelectFloatFromArray(Direct Array Index=$Particles.BubbleIndex, Float Selection Array=DataInterface() {propertyValues {FloatData [1.5,-1.5]}})
        )
        B=6
       )
       Period=1
       Scale=5
       Bias=0
      )
      B=&Sine(Normalized Angle=$Emitter.Age, Period=1, Scale=1, Bias=0)
     )
     Y=0
     Z=0
    )) {isSetParameters=true}
   &ShapeLocation(
    MaskSpawnGroup=false
    Offset=&MergeHelper_VectorSwitch(
     0=<0 0 0>
     1=<0 0 20>
     2=&Add_Vector(
      A=$Particles.SideMovement
      B=&Lerp_Vector(
       A=<0 0 120>
       B=&MaskVector(Mask Source=1, Mask Parameter=$Particles.BubbleIndex, Mask=0, Masked Vector=<40 0 150>, Passthrough Vector=<-40 0 150>)
       Alpha=$User.ParaGlideLeanAlpha
      )
     )
     3=?
     Index=$Particles.BubblePos
     Num=::NewEnumerator1
    )
    Offset Mode=::NewEnumerator0
    Override Seed=false
    Shape Origin=$Particles.SkeletalMeshLocation.SampledPosition
    Sphere Radius=&MergeHelper_FloatSwitch(0=15, 1=10, 2=15, 3=?, Index=$Particles.BubblePos, Num=::NewEnumerator1)
    Sphere Surface Distribution=1
    UseOwnerScale=true
   )
   /SetVariables(
    Particles.PopColor:LinearColor=<rgba 1 1 1 10>
    Particles.PopSize:Vector2f=&Multiply_Vector2D_ByFloat(
     Float=&FloatFromCurve(FloatCurve=DataInterface() {propertyValues {Curve {keys [
          <at 0 0 RCIM_Constant>
          <at 0.5 1 RCIM_Cubic>
          {arriveTangent=-6.2495, interpMode=RCIM_Cubic, leaveTangent=-6.2495, tangentMode=RCTM_User, time=1, value=0}
         ]}}})
     Vector2D=$Particles.SpriteSize
    )
   ) {isSetParameters=true}
   &ScaleSpriteSize(Uniform Curve Sprite Scale=DataInterface() {propertyValues {Curve {keys [
        {arriveTangent=9.2341, interpMode=RCIM_Cubic, leaveTangent=9.2341, tangentMode=RCTM_User, time=0, value=0}
        <at 0.15 1 RCIM_Constant>
        <at 0.5 0 RCIM_Constant>
       ]}}})
   &GravityForce(Gravity=<0 0 980>)
   &LinearForce(Force=&MergeHelper_VectorSwitch(
     0=&Lerp_Vector(
      A=<0 0 0>
      B=&MaskVector(Mask Source=1, Mask Parameter=$Particles.BubbleIndex, Mask=1, Masked Vector=<-250 0 0>, Passthrough Vector=<250 0 0>)
      Alpha=$User.ParaGlideLeanAlpha
     )
     1=&Lerp_Vector(
      A=<0 0 5000>
      B=&MaskVector(Mask Source=1, Mask Parameter=$Particles.BubbleIndex, Mask=0, Masked Vector=<2500 0 7500>, Passthrough Vector=<-2500 0 7500>)
      Alpha=$User.ParaGlideLeanAlpha
     )
     2=<0 0 0>
     3=?
     Index=$Particles.BubblePos
     Num=::NewEnumerator1
    ))
   &OffsetPosition(Position Offset=&Multiply_VectorByFloat(
     Vector=$Particles.SideMovement
     Float=&SelectFloatFromArray(Direct Array Index=$Particles.BubblePos, Float Selection Array=DataInterface() {propertyValues {FloatData [0,1,0]}})
    ))
   &VortexForce(
    UseOriginOffset=false
    Vortex Force Amount=&SelectFloatFromArray(Direct Array Index=$Particles.BubblePos, Float Selection Array=DataInterface() {propertyValues {FloatData [0,0,0]}})
    Vortex Origin:NiagaraPosition=$Particles.SkeletalMeshLocation.SampledPosition
   )
   &SolveForcesAndVelocity(Clamp Velocity=false, Limit Acceleration=false)
   /SetVariables(
    Particles.NoiseColor:LinearColor=&MakeLinearColorFromVectorAndFloat(
     "Vector (RGB)"=<1 1 1>
     "Float (Alpha)"=&FloatFromCurve(FloatCurve=DataInterface() {propertyValues {Curve {
         defaultValue=1
         keys [
          {interpMode=RCIM_Cubic, tangentMode=RCTM_User, time=0, value=0}
          {interpMode=RCIM_Cubic, tangentMode=RCTM_User, time=0.5, value=1}
          {tangentMode=RCTM_User, time=1, value=0}
         ]
        }}})
    )
    Particles.NoiseSize:Vector2f=&Multiply_Vector2D_ByFloat(Float=1, Vector2D=$Particles.Initial.SpriteSize)
   ) {isSetParameters=true}
   /SetVariables(
    Particles.Color=&MultiplyLinearColor(
     Linear Color=<rgba 1 1 1 1>
     Scale Factor=&MakeVector4(
      X=1
      Y=1
      Z=1
      W=&MaskFloat(
       Mask Source=1
       Mask Parameter=$Particles.BubblePos
       Mask=2
       Masked Float=&FloatFromCurve(FloatCurve=DataInterface() {propertyValues {Curve {keys [<at 0.75 0>,<at 1.5 1>]}}}, CurveIndex=$Emitter.Age)
       Passthrough Float=1
      )
     )
    )
    Particles.NoiseColor:LinearColor=&MultiplyLinearColor(
     Linear Color=$Particles.NoiseColor
     Scale Factor=&MakeVector4(
      X=1
      Y=1
      Z=1
      W=&MaskFloat(
       Mask Source=1
       Mask Parameter=$Particles.BubblePos
       Mask=2
       Masked Float=&FloatFromCurve(FloatCurve=DataInterface() {propertyValues {Curve {keys [<at 0.75 0>,<at 1.5 1>]}}}, CurveIndex=$Emitter.Age)
       Passthrough Float=1
      )
     )
    )
    Particles.PopColor:LinearColor=&MultiplyLinearColor(
     Linear Color=$Particles.PopColor
     Scale Factor=&MakeVector4(
      X=1
      Y=1
      Z=1
      W=&MaskFloat(
       Mask Source=1
       Mask Parameter=$Particles.BubblePos
       Mask=2
       Masked Float=&FloatFromCurve(FloatCurve=DataInterface() {propertyValues {Curve {keys [<at 0.75 0>,<at 1.5 1>]}}}, CurveIndex=$Emitter.Age)
       Passthrough Float=1
      )
     )
    )
    Particles.SpriteSize=&Multiply_Vector2D_ByFloat(
     Float=&MaskFloat(
      Mask Source=1
      Mask Parameter=$Particles.BubblePos
      Mask=2
      Masked Float=&FloatFromCurve(FloatCurve=DataInterface() {propertyValues {Curve {keys [<at 0.75 0>,<at 1.5 1>]}}}, CurveIndex=$Emitter.Age)
      Passthrough Float=1
     )
     Vector2D=$Particles.SpriteSize
    )
   ) {isSetParameters=true}
   /SetVariables(
    Particles.Color=&MultiplyLinearColor(Linear Color=$Particles.Color, Scale Factor=&MakeVector4_FromLinearColor(LinearColor=$System.TOD))
    Particles.NoiseColor:LinearColor=&MultiplyLinearColor(Linear Color=$Particles.NoiseColor, Scale Factor=&MakeVector4_FromLinearColor(LinearColor=$System.TOD))
    Particles.PopColor:LinearColor=&MultiplyLinearColor(Linear Color=$Particles.PopColor, Scale Factor=&MakeVector4_FromLinearColor(LinearColor=$System.TOD))
   ) {isSetParameters=true}
  ]
  rendererClasses [&NiagaraSpriteRendererProperties]
  renderers [
   {rendererClass=&NiagaraSpriteRendererProperties}
   {rendererClass=&NiagaraSpriteRendererProperties, rendererIndex=1}
   {rendererClass=&NiagaraSpriteRendererProperties, rendererIndex=2}
   {rendererClass=&NiagaraSpriteRendererProperties, rendererIndex=3}
  ]
 }
 CloudLarge {
  emitterUpdateScript [
   &EmitterState(Inactive Response=::NewEnumerator1, Life Cycle Mode=::NewEnumerator1, UseLoopDelay=false)
   &SpawnBurst_Instantaneous(
    Spawn Count=&MakeCustomIntFromBool(Bool=$System.Spawning, True INT=3, False INT=0)
    Use Spawn Probability=false
    UseLoopCountLimit=false
   )
   &SpawnBurst_Instantaneous(
    Spawn Count=&MakeCustomIntFromBool(Bool=$System.Spawning, True INT=3, False INT=0)
    Spawn Group=1
    Spawn Time=2
    Use Spawn Probability=false
    UseLoopCountLimit=false
   ) {module=SpawnBurst_Instantaneous001}
   &SpawnBurst_Instantaneous(
    Spawn Count=&MakeCustomIntFromBool(Bool=$System.Spawning, True INT=3, False INT=0)
    Spawn Group=2
    Spawn Time=3
    Use Spawn Probability=false
    UseLoopCountLimit=false
   ) {module=SpawnBurst_Instantaneous002}
   &SpawnBurst_Instantaneous(
    Spawn Count=&MakeCustomIntFromBool(Bool=$System.Spawning, True INT=3, False INT=0)
    Spawn Group=3
    Spawn Time=5
    Use Spawn Probability=false
    UseLoopCountLimit=false
   ) {module=SpawnBurst_Instantaneous003}
   &SpawnBurst_Instantaneous(
    Spawn Count=&MakeCustomIntFromBool(Bool=$System.Spawning, True INT=3, False INT=0)
    Spawn Group=4
    Spawn Time=7
    Use Spawn Probability=false
    UseLoopCountLimit=false
   ) {module=SpawnBurst_Instantaneous004}
   &SpawnBurst_Instantaneous(
    Spawn Count=&MakeCustomIntFromBool(Bool=$System.Spawning, True INT=3, False INT=0)
    Spawn Group=5
    Spawn Time=8.5
    Use Spawn Probability=false
    UseLoopCountLimit=false
   ) {module=SpawnBurst_Instantaneous005}
  ]
  particleSpawnScript [
   &InitializeParticle(
    Color=&SelectLinearColorFromArray(
     Array Sampling Mode=::NewEnumerator0
     Direct Array Index=?
     Color Selection Array=DataInterface() {propertyValues {ColorData [<rgba 0.6979 0.459 0.2792 1>, <rgba 0.4669 0.2666 0.8125 1>, <rgba 0.5782 0.1314 0.9792 1>, <rgba 0.2101 0.4891 1 1>]}}
     Random Seed=0
    )
    Lifetime=5
    Sprite Rotation Angle=&RandomRangeFloat(Minimum=0, Maximum=360)
    Sprite Rotation Mode=::NewEnumerator0
    Sprite Size Mode=::NewEnumerator3
    Uniform Sprite Size=&Multiply_Float(
     A=90
     B=&SelectFloatFromArray(Direct Array Index=$Emitter.SpawnGroup, Float Selection Array=DataInterface() {propertyValues {FloatData [1,1.3,1.5,1,1.3,1.5]}})
    )
    UsePositionOffset=false
    Write Material Random=true
   ) {module=InitializeParticle001}
   &ShapeLocation(
    MaskSpawnGroup=false
    Offset=&FN_FrontEndVector3Modifier(Gameplay=<0 0 -300>, Front End=<0 0 -125>)
    Offset Mode=::NewEnumerator0
    Override Seed=false
    Ring / Disc Distribution Mode:ENiagaraSphereDistributionMode=::NewEnumerator1
    Ring Radius=200
    Shape Primitive=::NewEnumerator4
    U Position=&SelectFloatFromArray(Direct Array Index=$Emitter.SpawnGroup, Float Selection Array=DataInterface() {propertyValues {FloatData [0.1,0.5,0.3,0.8,0.2,0.9]}})
    UseOwnerScale=true
   ) {module=ShapeLocation001}
   &ShapeLocation(
    Cylinder Height=1
    Cylinder Radius=&SelectFloatFromArray(Direct Array Index=$Emitter.SpawnGroup, Float Selection Array=DataInterface() {propertyValues {FloatData [15,22,25,15,20,22]}})
    MaskSpawnGroup=false
    Non Uniform Scale=<2.5 0.7 0.7>
    Override Seed=false
    Shape Primitive=::NewEnumerator1
    Surface Only=true
    UseOwnerScale=true
   )
   &AddVelocity(Velocity=<0 -35 100>, Velocity Speed Scale=&FN_FrontEndFloatModifier(Gameplay=0, Front End=1))
  ]
  particleUpdateScript [
   &ParticleState()
   &SpriteRotationRate(Rotation Rate=&Multiply_Float(
     A=&RandomRangeFloat(Minimum=5, Maximum=10)
     B=&SelectFloatFromArray(
      Array Sampling Mode=::NewEnumerator0
      Direct Array Index=?
      Float Selection Array=DataInterface() {propertyValues {FloatData [-1,1]}}
      Random Seed=0
     )
    ))
   &GravityForce(
    Coordinate Space=::Local
    Gravity=&MakeVector(
     X=0
     Y=&FN_FrontEndFloatModifier(
      Gameplay=0
      Front End=&RemapRange(Input Value To Remap=$User.ParaGlideLeanAlpha, Input Min=0, Input Max=1, Output Min=0, Output Max=-50)
     )
     Z=&FN_FrontEndFloatModifier(
      Gameplay=0
      Front End=&RemapRange(Input Value To Remap=$User.ParaGlideLeanAlpha, Input Min=0, Input Max=1, Output Min=0, Output Max=150)
     )
    )
   ) {module=GravityForce001}
   &InheritSourceMovement(
    Applied Position Scale=&VectorFromFloat(Value=&FloatFromCurve(FloatCurve=DataInterface() {propertyValues {Curve {keys [
          {arriveTangent=-0.3402, interpMode=RCIM_Cubic, leaveTangent=-0.3402, tangentMode=RCTM_User, time=0, value=1}
          {arriveTangent=-1.6874, interpMode=RCIM_Cubic, leaveTangent=-1.6874, tangentMode=RCTM_User, time=1, value=0.2}
         ]}}}))
    ApplyForce=false
    ApplyPosition=true
    ApplyPositionFalloff=false
   )
   &SolveForcesAndVelocity(Clamp Velocity=false, Limit Acceleration=false)
   &ScaleColor(
    Scale Alpha=&FloatFromCurve(FloatCurve=DataInterface() {propertyValues {Curve {keys [
         {arriveTangent=1.4796, interpMode=RCIM_Cubic, leaveTangent=1.4796, tangentMode=RCTM_User, time=0, value=0}
         {interpMode=RCIM_Cubic, tangentMode=RCTM_User, time=0.5, value=1}
         {arriveTangent=-2.9378, interpMode=RCIM_Cubic, leaveTangent=-2.9378, tangentMode=RCTM_User, time=1, value=0}
        ]}}})
    ScaleA=true
    ScaleRGB=true
   )
   &DynamicMaterialParameters(
    Index 0 Param 1=&FloatFromCurve(FloatCurve=DataInterface() {propertyValues {Curve {keys [<at 0 0.4824 RCIM_Cubic>, {arriveTangent=0.5176, interpMode=RCIM_Cubic, leaveTangent=0.5176, time=0.5, value=0}, <at 1 1 RCIM_Cubic>]}}})
    Index 0 Param 2=1
    Index 0 Param 3=1
    Module.Index0Float0Write=true
    Module.Index0Float1Write=true
    Module.Index0Float2Write=true
    Module.Index0Float3Write=true
    Module.Index1Float0Write=true
    Module.Index1Float1Write=true
    Module.Index1Float2Write=true
    Module.Index1Float3Write=true
    Write Parameter Index 1=::NewEnumerator1
   ) {module=DynamicMaterialParameters001}
   /SetVariables(Particles.Color=&MultiplyLinearColor(Linear Color=$Particles.Color, Scale Factor=&MakeVector4_FromLinearColor(LinearColor=$System.TOD))) {isSetParameters=true}
  ]
  rendererClasses [&NiagaraSpriteRendererProperties]
  renderers [{rendererClass=&NiagaraSpriteRendererProperties}]
 }
 Clouds {
  emitterUpdateScript [
   &EmitterState()
   &SpawnBurst_Instantaneous(Spawn Count=15, Use Spawn Probability=false, UseLoopCountLimit=true)
   &SpawnRate(SpawnRate=30, Use Spawn Probability=false)
  ]
  particleSpawnScript [&InitializeParticle(
    Color=&SelectLinearColorFromArray(
     Array Sampling Mode=::NewEnumerator0
     Direct Array Index=?
     Color Selection Array=DataInterface() {propertyValues {ColorData [<rgba 0.6979 0.3897 0.1578 1>, <rgba 0.3697 0.113 0.8125 1>, <rgba 0.5442 0.0596 0.9792 1>, <rgba 0.0957 0.232 1 1>]}}
     Random Seed=0
    )
    Lifetime=2
    Sprite Size=&Multiply_Vector2D_ByFloat(Float=1, Vector2D=<20 15>)
    Sprite Size Mode=::NewEnumerator4
    Sprite UV Mode=::NewEnumerator1
    UsePositionOffset=false
    Write Material Random=true
   )]
  particleUpdateScript [
   &ParticleState()
   &SkeletalMeshLocation(
    Bone / Socket Index=&RandomRangeInt(Minimum=0, Maximum=1)
    Bone Sampling Mode=::NewEnumerator4
    MaskSpawnGroup=false
    Skeletal Mesh=$System.Skeletal Mesh
   )
   &ShapeLocation(
    Additional Quaternion Rotation=true
    Additional Yaw / Pitch / Roll=<0 90 0>
    Handle Radius=5
    Large Radius=10
    MaskSpawnGroup=false
    Override Seed=false
    Rotation Angle Type:ENiagara_AngleInput=::NewEnumerator0
    Rotation Mode=::NewEnumerator3
    Rotation Quaternion=$Particles.SkeletalMeshLocation.SampledRotation
    Shape Origin=$Particles.SkeletalMeshLocation.SampledPosition
    Shape Primitive=::NewEnumerator3
    UseOwnerScale=true
   )
   &SubUVAnimation(
    "Sprite Renderer "=DataInterface() {propertyValues {SpriteRenderer=&self:Clouds_0.NiagaraSpriteRendererProperties_0}}
    UseEndFrame=false
    UseStartFrame=false
   )
   &DynamicMaterialParameters(
    Index 0 Param 1=&FloatFromCurve(FloatCurve=DataInterface() {propertyValues {Curve {keys [<at 0 1 RCIM_Cubic>,<at 0.5 0 RCIM_Cubic>,<at 1 1 RCIM_Cubic>]}}})
    Index 0 Param 2=0.5
    Index 0 Param 3=1
    Module.Index0Float0Write=true
    Module.Index0Float1Write=true
    Module.Index0Float2Write=true
    Module.Index0Float3Write=true
    Module.Index1Float0Write=true
    Module.Index1Float1Write=true
    Module.Index1Float2Write=true
    Module.Index1Float3Write=true
    Write Parameter Index 1=::NewEnumerator1
   )
   &ScaleColor(Scale RGB=<2 2 2>, ScaleA=true, ScaleRGB=true)
   /SetVariables(Particles.Color=&MultiplyLinearColor(Linear Color=$Particles.Color, Scale Factor=&MakeVector4_FromLinearColor(LinearColor=$System.TOD))) {isSetParameters=true}
  ]
  rendererClasses [&NiagaraSpriteRendererProperties]
  renderers [{rendererClass=&NiagaraSpriteRendererProperties}]
 }
 MainTrail {
  emitterUpdateScript [
   &EmitterState(Life Cycle Mode=::NewEnumerator1, Loop Behavior=::NewEnumerator1, Loop Duration=1, UseLoopDelay=false)
   &SpawnBurst_Instantaneous(Spawn Count=400, Use Spawn Probability=false, UseLoopCountLimit=false)
  ]
  particleSpawnScript [
   &InitializeParticle(
    Lifetime=5
    Ribbon Width=25
    Ribbon Width Mode=::NewEnumerator1
    Sprite Size Mode=::NewEnumerator3
    UsePositionOffset=false
    Write Material Random=true
   )
   /SetVariables(Particles.SpiralFreq2=0.5) {isSetParameters=true}
  ]
  particleUpdateScript [
   &ParticleState(Kill Particles When Lifetime Has Elapsed=false)
   &PartitionParticles(
    Number Of Partitions=4
    Particles Per Partition:NiagaraInt32=&SelectIntFromArray(Direct Array Index=$Engine.QualityLevel, Int Selection Array=DataInterface() {propertyValues {IntData [25,50,100,100,100]}})
    Partition Mode:ENiagara_PartitionMode_v2=::NewEnumerator1
   )
   &SkeletalMeshLocation(
    Bone / Socket Index=&SelectIntFromArray(Direct Array Index=$Particles.PartitionParticles.Partition, Int Selection Array=DataInterface() {propertyValues {IntData [0,0,1,1]}})
    Bone Sampling Mode=::NewEnumerator4
    MaskSpawnGroup=false
    Skeletal Mesh=$System.Skeletal Mesh
   )
   /SetVariables(Particles.Height:Vector3f=&AddVectorToPosition(
     Position=$Particles.SkeletalMeshLocation.SampledPosition
     Vector=&VectorFromCurve(
      VectorCurve=DataInterface() {propertyValues {XCurve {keys [<at 0 0>,<at 1 0>]}, YCurve {keys [<at 0 0>,<at 1 0>]}, ZCurve {keys [<at 0 0>,<at 1 1>]}}}
      CurveIndex=$Particles.PartitionParticles.Normalized Index In Partition
      Scale Curve=&Lerp_Vector(A=<0 0 180>, B=<1 1 220>, Alpha=$User.ParaGlideLeanAlpha)
     )
    )) {isSetParameters=true}
   /SetVariables(Particles.SideMovement:Vector3f=&MakeVector(
     X=&Multiply_Float(
      A=&Sine(
       Normalized Angle=&Multiply_Float(
        A=&Add_Float(
         A=$Particles.PartitionParticles.Normalized Index In Partition
         B=&Multiply_Float(A=$Engine.System.Age, B=-0.75)
        )
        B=&SelectFloatFromArray(
         Direct Array Index=$Particles.PartitionParticles.Partition
         Float Selection Array=DataInterface() {propertyValues {FloatData [3,1.5,-3,-1.5]}}
        )
       )
       Period=1
       Scale=20
       Bias=&Lerp_Float(
        A=0
        B=&MaskFloat(Mask Source=1, Mask Parameter=$Particles.PartitionParticles.Partition, Mask=0, Masked Float=50, Passthrough Float=-50)
        Alpha=$User.ParaGlideLeanAlpha
       )
      )
      B=&Add_Float(A=$Particles.PartitionParticles.Normalized Index In Partition, B=0)
     )
     Y=&Multiply_Float(
      A=&Cosine(
       Normalized Angle=&Multiply_Float(
        A=&Add_Float(
         A=$Particles.PartitionParticles.Normalized Index In Partition
         B=&Multiply_Float(A=$Engine.System.Age, B=-0.75)
        )
        B=&SelectFloatFromArray(
         Direct Array Index=$Particles.PartitionParticles.Partition
         Float Selection Array=DataInterface() {propertyValues {FloatData [3,1.5,-3,-1.5]}}
        )
       )
       Period=1
       Scale=-20
       Bias=&Lerp_Float(A=0, B=1, Alpha=$User.ParaGlideLeanAlpha)
      )
      B=$Particles.PartitionParticles.Normalized Index In Partition
     )
     Z=0
    )) {isSetParameters=true}
   /SetVariables(Particles.FlatPosition:NiagaraPosition=&AddVectorToPosition(Position=Linked() {linkedVariable {name=Particles.Height, type=Vector3f}}, Vector=$Particles.SideMovement)) {isSetParameters=true}
   /SetVariables(Particles.SpiralStrength2:NiagaraFloat=&Multiply_Float(
     A=9
     B=&Lerp_Float(
      A=&FloatFromCurve(
       FloatCurve=DataInterface() {propertyValues {Curve {
          defaultValue=1.625
          keys [
           {arriveTangent=2.0886, interpMode=RCIM_Cubic, leaveTangent=2.0886, tangentMode=RCTM_User, time=0.1, value=0.1882}
           {arriveTangent=-0.2551, interpMode=RCIM_Cubic, leaveTangent=-0.2551, tangentMode=RCTM_User, time=1, value=0.4941}
          ]
         }}}
       CurveIndex=$Particles.PartitionParticles.Normalized Index In Partition
       Scale Curve=2
      )
      B=&FloatFromCurve(
       FloatCurve=DataInterface() {propertyValues {Curve {
          defaultValue=1.625
          keys [
           {arriveTangent=-0.3372, interpMode=RCIM_Cubic, leaveTangent=-0.3372, tangentMode=RCTM_User, time=0.1, value=0.2}
           {arriveTangentWeight=0.2963, interpMode=RCIM_Cubic, leaveTangentWeight=0.127, tangentMode=RCTM_User, time=0.5015, value=0.5}
           {arriveTangent=-0.3467, interpMode=RCIM_Cubic, leaveTangent=-0.3467, tangentMode=RCTM_User, time=1, value=0}
          ]
         }}}
       CurveIndex=$Particles.PartitionParticles.Normalized Index In Partition
       Scale Curve=2
      )
      Alpha=$User.ParaGlideLeanAlpha
     )
    )) {isSetParameters=true}
   /SetVariables(Particles.SpiralTime2:NiagaraFloat=&Add_Float(
     A=&Add_Float(A=&Multiply_Float(A=$System.Age, B=-0.5), B=$Particles.PartitionParticles.Normalized Index In Partition)
     B=&Multiply_FloatByInt(
      Float=&Divide_Float(A=$Particles.SpiralFreq2, B=&MakeFloatFromInt(Int=3))
      Integer=$Particles.PartitionParticles.Partition
     )
    )) {isSetParameters=true}
   /SetVariables(Particles.SpiralOffset2:Vector3f=&MakeVector(
     X=&Sine(
      Normalized Angle=$Particles.SpiralTime2
      Period=$Particles.SpiralFreq2
      Scale=&Multiply_Float(A=$Particles.SpiralStrength2, B=1.5)
      Bias=&Lerp_Float(
       A=0
       B=&Multiply_Float(
        A=$Particles.PartitionParticles.Normalized Index In Partition
        B=&MaskFloat(Mask Source=1, Mask Parameter=$Particles.PartitionParticles.Partition, Mask=1, Masked Float=50, Passthrough Float=-50)
       )
       Alpha=$User.ParaGlideLeanAlpha
      )
     )
     Y=&Cosine(
      Normalized Angle=$Particles.SpiralTime2
      Period=$Particles.SpiralFreq2
      Scale=&Multiply_Float(A=$Particles.SpiralStrength2, B=1.5)
      Bias=0
     )
     Z=0
    )) {isSetParameters=true}
   /SetVariables(Particles.Positon2:NiagaraPosition=&AddVectorToPosition(Position=Linked() {linkedVariable {name=Particles.Height, type=Vector3f}}, Vector=$Particles.SpiralOffset2)) {isSetParameters=true}
   /SetVariables(Particles.RibbonID=$Output.PartitionParticles.RibbonID) {isSetParameters=true}
   &ScaleColor(
    Scale Alpha=&FloatFromCurve(
     FloatCurve=DataInterface() {propertyValues {Curve {keys [<at 0.4309 1>,<at 0.7908 0>]}}}
     CurveIndex=$Particles.PartitionParticles.Normalized Index In Partition
    )
    Scale RGB=&Multiply_VectorByFloat(
     Vector=<1 1 1>
     Float=&FloatFromCurve(
      FloatCurve=DataInterface() {propertyValues {Curve {keys [
          <at 0.1119 1 RCIM_Cubic>
          {arriveTangent=-0.73, interpMode=RCIM_Cubic, leaveTangent=-0.73, time=0.3453, value=0.4265}
          {arriveTangent=1.2738, interpMode=RCIM_Cubic, leaveTangent=1.2738, time=0.5108, value=0.7088}
          {arriveTangent=-1.4489, interpMode=RCIM_Cubic, leaveTangent=-1.4489, time=0.7517, value=0.9441}
          <at 1 0 RCIM_Cubic>
         ]}}}
      CurveIndex=$Particles.PartitionParticles.Normalized Index In Partition
     )
    )
    ScaleA=true
    ScaleRGB=false
   )
   &DynamicMaterialParameters(
    Index 0 Param 1=-20
    Index 0 Param 4=0.3
    Module.Index0Float0Write=true
    Module.Index0Float1Write=true
    Module.Index0Float2Write=true
    Module.Index0Float3Write=true
    Module.Index1Float0Write=true
    Module.Index1Float1Write=true
    Module.Index1Float2Write=true
    Module.Index1Float3Write=true
    Write Parameter Index 1=::NewEnumerator1
   )
   &ScaleRibbonWidth(Ribbon Width Scale=&FloatFromCurve(
     FloatCurve=DataInterface() {propertyValues {Curve {
        defaultValue=1
        keys [
         {interpMode=RCIM_Cubic, tangentMode=RCTM_User, time=0, value=0.25}
         {interpMode=RCIM_Cubic, tangentMode=RCTM_User, time=0.5, value=0.7353}
         {tangentMode=RCTM_User, time=1, value=0.6647}
        ]
       }}}
     CurveIndex=$Particles.PartitionParticles.Normalized Index In Partition
     Scale Curve=&SelectFloatFromArray(Direct Array Index=$Particles.PartitionParticles.Partition, Float Selection Array=DataInterface() {propertyValues {FloatData [1,0,1,0]}})
    ))
   /SetVariables(
    Particles.RibbonColor:LinearColor=&MakeLinearColorFromVectorAndFloat(
     "Vector (RGB)"=<1.5 1.5 1.5>
     "Float (Alpha)"=&FloatFromCurve(
      FloatCurve=DataInterface() {propertyValues {Curve {keys [<at 0.6594 1>,<at 0.8727 0>]}}}
      CurveIndex=$Particles.PartitionParticles.Normalized Index In Partition
     )
    )
    Particles.RibbonSize:NiagaraFloat=&MaskFloat(
     Mask Source=1
     Mask Parameter=$Particles.PartitionParticles.Partition
     Mask=1
     Masked Float=&FloatFromCurve(
      FloatCurve=DataInterface() {propertyValues {Curve {
         defaultValue=1
         keys [
          {interpMode=RCIM_Cubic, tangentMode=RCTM_User, time=0, value=0.25}
          {interpMode=RCIM_Cubic, tangentMode=RCTM_User, time=0.5, value=1}
          {tangentMode=RCTM_User, time=1, value=-0.5}
         ]
        }}}
      CurveIndex=$Particles.PartitionParticles.Normalized Index In Partition
      Scale Curve=&FloatFromCurve(
       FloatCurve=DataInterface() {propertyValues {Curve {defaultValue=1, keys [{tangentMode=RCTM_User, time=0.3501, value=0}, {tangentMode=RCTM_User, time=0.7472, value=1}]}}}
       CurveIndex=&Multiply_Float(A=&OneMinusFloat(Float=$Particles.PartitionParticles.Normalized Index In Partition), B=$Particles.Age)
       Scale Curve=5
      )
     )
     Passthrough Float=&MaskFloat(
      Mask Source=1
      Mask Parameter=$Particles.PartitionParticles.Partition
      Mask=3
      Masked Float=&FloatFromCurve(
       FloatCurve=DataInterface() {propertyValues {Curve {
          defaultValue=1
          keys [
           {interpMode=RCIM_Cubic, tangentMode=RCTM_User, time=0, value=0.25}
           {interpMode=RCIM_Cubic, tangentMode=RCTM_User, time=0.5, value=1}
           {tangentMode=RCTM_User, time=1, value=-0.5}
          ]
         }}}
       CurveIndex=$Particles.PartitionParticles.Normalized Index In Partition
       Scale Curve=&FloatFromCurve(
        FloatCurve=DataInterface() {propertyValues {Curve {defaultValue=1, keys [{tangentMode=RCTM_User, time=0.3501, value=0}, {tangentMode=RCTM_User, time=0.7472, value=1}]}}}
        CurveIndex=&Multiply_Float(A=&OneMinusFloat(Float=$Particles.PartitionParticles.Normalized Index In Partition), B=$Particles.Age)
        Scale Curve=5
       )
      )
      Passthrough Float=0
     )
    )
   ) {isSetParameters=true}
   &ScaleColor(
    Scale Alpha=&FloatFromCurve(
     FloatCurve=DataInterface() {propertyValues {Curve {defaultValue=1, keys [{tangentMode=RCTM_User, time=0.3501, value=0}, {tangentMode=RCTM_User, time=0.7472, value=1}]}}}
     CurveIndex=&Multiply_Float(A=&OneMinusFloat(Float=$Particles.PartitionParticles.Normalized Index In Partition), B=$Particles.Age)
    )
    ScaleA=true
    ScaleRGB=true
   ) {module=ScaleColor001}
   /SetVariables(
    Particles.Color=&MultiplyLinearColor(Linear Color=$Particles.Color, Scale Factor=&MakeVector4_FromLinearColor(LinearColor=$System.TOD))
    Particles.RibbonColor:LinearColor=&MultiplyLinearColor(Linear Color=$Particles.RibbonColor, Scale Factor=&MakeVector4_FromLinearColor(LinearColor=$System.TOD))
   ) {isSetParameters=true}
  ]
  rendererClasses [&NiagaraRibbonRendererProperties]
  renderers [
   {rendererClass=&NiagaraRibbonRendererProperties}
   {rendererClass=&NiagaraRibbonRendererProperties, rendererIndex=1}
  ]
 }
 Ribbons001 {
  emitterUpdateScript [
   &EmitterState()
   &SpawnRate(SpawnRate=1.5, Use Spawn Probability=false)
  ]
  particleSpawnScript [
   &FN_SetSpawnGroup(Number of Spawn Groups=2)
   &PartitionParticles()
   &InitializeParticle(
    Color=&SelectLinearColorFromArray(
     Array Sampling Mode=::NewEnumerator0
     Direct Array Index=?
     Color Selection Array=DataInterface() {propertyValues {ColorData [<rgba 1 0.2174 0.6256 3>, <rgba 0.7552 0.6545 0.1116 3>, <rgba 0.6667 0.2965 0.1507 3>, <rgba 0.3584 0.4995 0.687 3>]}}
     Random Seed=0
    )
    Lifetime Max=3
    Lifetime Min=2
    Lifetime Mode=::NewEnumerator1
    Mass Mode=::NewEnumerator1
    Mesh Scale=&RandomRangeVector(Minimum=<1.2 1.2 1.1>, Maximum=<1.6 1.6 1.3>)
    Mesh Scale Mode=::NewEnumerator4
    Sprite Size Mode=::NewEnumerator4
    UsePositionOffset=false
    Write Material Random=true
   ) {module=InitializeParticle001}
   &InitialMeshOrientation(
    Mesh Orientation Mode=::NewEnumerator2
    Rotation=&RandomRangeVector(Minimum=<0 0 0>, Maximum=<0 0 360>)
    Use Rotation Vector=true
   )
  ]
  particleUpdateScript [
   &ParticleState()
   &ScaleMeshSize(Scale Factor=&MakeVector(
     X=1
     Y=1
     Z=&FloatFromCurve(
      FloatCurve=DataInterface() {propertyValues {Curve {defaultValue=1, keys [{tangentMode=RCTM_User, time=0.2972, value=0}, {tangentMode=RCTM_User, time=1.2972, value=1}]}}}
      CurveIndex=$Emitter.Age
      Scale Curve=&Lerp_Float(A=1, B=1.2, Alpha=$User.ParaGlideLeanAlpha)
     )
    ))
   &UpdateMeshOrientation(
    Rotation Rate=&SelectFloatFromArray(
     Direct Array Index=$Particles.Initial.PartitionParticles.Partition
     Float Selection Array=DataInterface() {propertyValues {FloatData [1,-1]}}
    )
    Rotation Vector=<0 0 1>
   )
   &UpdateMeshOrientation(
    Facing Direction=&Lerp_Vector(
     A=<1 0 0>
     B=&SelectVectorFromArray(
      Direct Array Index=$Particles.PartitionParticles.Partition
      Vector Selection Array=DataInterface() {propertyValues {FloatData [<1 0 -0.2>, <1 0 0.2>]}}
     )
     Alpha=$User.ParaGlideLeanAlpha
    )
    Facing Interpolation=false
    Orientation Method=::NewEnumerator1
    Override Quat=false
    TransformFacing=false
    TransformUp=false
   ) {module=UpdateMeshOrientation001}
   &SkeletalMeshLocation(
    Bone / Socket Index=$Particles.SpawnGroup
    Bone Sampling Mode=::NewEnumerator4
    Inherit Velocity=false
    Interpolate Mesh Sampling=false
    MaskSpawnGroup=false
    Skeletal Mesh=$System.Skeletal Mesh
   )
   &SolveForcesAndVelocity(Clamp Velocity=false, Limit Acceleration=false)
   &ScaleColor(
    Scale Alpha=&FloatFromCurve(
     FloatCurve=DataInterface() {propertyValues {Curve {
        defaultValue=1
        keys [
         {interpMode=RCIM_Cubic, tangentMode=RCTM_User, time=0, value=0}
         {arriveTangentWeight=0.2835, interpMode=RCIM_Cubic, leaveTangentWeight=0.0892, tangentMode=RCTM_User, time=0.25, value=1}
         {interpMode=RCIM_Cubic, tangentMode=RCTM_User, time=0.75, value=1}
         {tangentMode=RCTM_User, time=1, value=0}
        ]
       }}}
     Scale Curve=&FloatFromCurve(
      FloatCurve=DataInterface() {propertyValues {Curve {
         defaultValue=1
         keys [{interpMode=RCIM_Cubic, tangentMode=RCTM_User, time=0.2892, value=0}, {interpMode=RCIM_Cubic, tangentMode=RCTM_User, time=1.2892, value=1}]
        }}}
      CurveIndex=$Emitter.Age
     )
    )
    ScaleA=true
    ScaleRGB=true
   )
   /SetVariables(Particles.Color=&MultiplyLinearColor(Linear Color=$Particles.Color, Scale Factor=&MakeVector4_FromLinearColor(LinearColor=$System.TOD))) {isSetParameters=true}
  ]
  rendererClasses [&NiagaraMeshRendererProperties]
  renderers [{rendererClass=&NiagaraMeshRendererProperties, rendererIndex=0}]
 }
}
summary {
 systemName=NS_Contrail_SteedShimmer
 emitters [
  {emitterName=Bubbles_Stars, rendererClasses [&NiagaraSpriteRendererProperties]}
  {emitterName=Clouds, rendererClasses [&NiagaraSpriteRendererProperties]}
  {emitterName=MainTrail, rendererClasses [&NiagaraRibbonRendererProperties]}
  {emitterName=CloudLarge, rendererClasses [&NiagaraSpriteRendererProperties]}
  {emitterName=Ribbons001, rendererClasses [&NiagaraMeshRendererProperties]}
 ]
 userVariables [
  User.bIsFrontEnd {type=NiagaraBool}
  User.bIsFrontEndPreview {type=NiagaraBool}
  User.ParaGlideLeanAlpha {type=NiagaraFloat, defaultValue=0, description=""}
  User.RibbonIDOffset {type=NiagaraInt32}
 ]
}
